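# UI-domain material for the PistolCanary player video, stored with the layout's media assets.
# It holds a single texture sample of PistolCanary_Player_Video, with no math or parameters.
# The material is unlit, and automatic usage flagging in the editor is off.
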Material /PistolCanary/Layout/Media/PistolCanary_Player_Video_Mat
graph {
 properties {bAutomaticallySetUsageInEditor=false, materialDomain=MD_UI, shadingModel=MSM_Unlit, usedShadingModels=Unlit}
 nodes [TextureSample() {texture=&/PistolCanary_Player_Video}]
}
